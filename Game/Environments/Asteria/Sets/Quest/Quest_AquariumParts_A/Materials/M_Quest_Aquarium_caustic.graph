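# Translucent deferred decal material that casts animated water caustics for the quest aquarium environment set.
# World position relative to a location is transformed by a parameterized 3x3 matrix into projected UVs, with optional perspective.
# Two panning caustic samples, distorted by a two-octave warp, blend between sharp and blurred textures based on water depth.
# Inverse-square, radial and height falloffs shape the tinted intensity, which also feeds an eye-adaptation divide and an opacity lerp.
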Material /Game/Environments/Asteria/Sets/Quest/Quest_AquariumParts_A/Materials/M_Quest_Aquarium_caustic
graph {
 properties {blendMode=BLEND_Translucent, materialDomain=MD_DeferredDecal, pixelDepthOffsetMode=PDOM_Legacy, usedShadingModels=Default Lit}
 nodes [
  ActorPositionWS()
  Constant()
  n1 = NamedRerouteUsage()
  n2 = ScalarParameter() {defaultValue=1, group=DFX, parameterName=DFX_Intensity}
  n3 = Divide(A=Distance(A=WorldPosition().XYZ, B=ActorPositionWS())) {constB=100, desc=CM to M}
  n4 = NamedRerouteDeclaration(None=Divide(B=Multiply(A=@n3, B=@n3))) {name=Inverse Square Law, nodeColor=<rgba 0.7812 1 0 1>}
  n5 = ScalarParameter() {defaultValue=1, group=DFX, parameterName=DFX_Falloff}
  n6 = TextureCoordinate()
  n7 = Constant2Vector() {g=0.5, r=0.5}
  n8 = ScalarParameter() {defaultValue=1, group=DFX, parameterName=DFX_FalloffRadius}
  n9 = Multiply(A=Multiply(A=LinearInterpolate(B=@n4, Alpha=@n5) {constA=1}, B=SphereMask(A=@n6, B=@n7, Radius=@n8) {hardnessPercent=0}))
  n10 = NamedRerouteUsage()
  n11 = OneMinus(None=Subtract(A=WorldPosition().Z, B=ComponentMask(None=@n10) {b=true}))
  n12 = ScalarParameter() {defaultValue=163.14, group=DFX, parameterName=DFX_FalloffRadius2}
  n13 = Power(Base=Saturate(None=Divide(A=@n11, B=@n12) {constB=100}), Exp=Constant() {r=2})
  n14 = Multiply(
   A=Multiply(A=Multiply(A=@n1, B=@n2), B=Multiply(A=@n9, B=@n13))
   B=VectorParameter() {group=DFX, parameterName=DFX_TintColor, defaultValue=<rgba 1 1 1 0>}.RGB
  )
  Divide(A=@n14, B=EyeAdaptation().EyeAdaptation)
  LinearInterpolate(A=@n14, Alpha=ScalarParameter() {defaultValue=0.3, group=Caustics, parameterName=caustic Opacity}) {constAlpha=0.4}
  NamedRerouteDeclaration() {name=water depth, nodeColor=<rgba 0 1 0.0859 1>}
  n15 = TextureObjectParameter() {
   group=DFX
   parameterName=DFX_Texture_blur
   samplerType=SAMPLERTYPE_LinearColor
   texture=/Game/Environments/Asteria/Sets/Quest/Quest_AquariumParts_A/Textures/T_Caustic_M_blur
  }
  NamedRerouteDeclaration(None=@n15) {name=TextureObject blur, nodeColor=<rgba 0 0.0156 1 1>}
  n16 = Subtract(
   A=LinearInterpolate(
    A=WorldPosition().XYZ
    B=WorldPosition() {worldPositionShaderOffset=WPT_ExcludeAllShaderOffsets}.XYZ
    Alpha=ScalarParameter() {defaultValue=1, group=DFX_Projection, parameterName=DFX_FixedTexture, sliderMax=1}
   )
   B=NamedRerouteUsage()
  )
  n17 = Normalize(VectorInput=@n16)
  n18 = MaterialFunctionCall(
   VectorToTransform=LinearInterpolate(
    A=@n16
    B=Multiply(A=@n17) {constB=100, desc="Arbitrary Mutiplier for\r\nbetter world scale"}
    Alpha=ScalarParameter() {group=DFX_Projection, parameterName=DFX_Perspective, sliderMax=1}
   )
   BasisX=VectorParameter() {group=DFX_Projection, parameterName=DFX_Matrix_X, defaultValue=<rgba 1 0 0 0>}.RGB
   BasisY=VectorParameter() {group=DFX_Projection, parameterName=DFX_Matrix_Y, defaultValue=<rgba 0 1 0 0>}.RGB
   BasisZ=VectorParameter() {group=DFX_Projection, parameterName=DFX_Matrix_Z, defaultValue=<rgba 0 0 1 0>}.RGB
  ) {materialFunction=&Transform3x3Matrix}
  n19 = Multiply(
   A=NamedRerouteDeclaration(None=ComponentMask(None=@n18.Transformed Vector) {g=true, r=true}) {name=UV_Centered, nodeColor=<rgba 0 1 0.25 1>}
   B=Constant2Vector() {g=-1, r=1}
  ) {constB=-1}
  n20 = ScalarParameter() {defaultValue=1, group=DFX_Projection, parameterName=DFX_Tiling}
  NamedRerouteDeclaration(None=Add(A=Multiply(A=@n19, B=@n20), B=ScalarParameter() {parameterName=UV Offset}) {constB=0.5}) {name=Coordinates, nodeColor=<rgba 1 0.8906 0 1>}
  NamedRerouteDeclaration(None=TextureObjectParameter() {group=DFX, parameterName=DFX_Texture, samplerType=SAMPLERTYPE_LinearColor, texture=/FortWater/Textures/T_Caustic_M}) {name=Texture Object, nodeColor=<rgba 0 0.0859 1 1>}
  NamedRerouteDeclaration(None=Add(A=ActorPositionWS(), B=VectorParameter() {parameterName=ACTOR position, defaultValue=<rgba 0 0 0 1>}.RGB)) {name=Location, nodeColor=<rgba 1 0 0.9141 1>}
  n21 = NamedRerouteUsage()
  n22 = ComponentMask(None=@n21) {g=true, r=true}
  n23 = NamedRerouteUsage()
  n24 = ConstantBiasScale(None=TextureSample(UVs=Panner(Coordinate=@n22) {speedX=-0.05, speedY=0.02}, Tex=@n23) {samplerSource=SSM_Wrap_WorldGroupSettings, samplerType=SAMPLERTYPE_LinearColor}.A) {bias=-0.5, scale=2}
  n25 = ScalarParameter() {defaultValue=0.0338, group=Caustics, parameterName=Caustic Warp Strength, sliderMax=0.1}
  n26 = ComponentMask(None=Divide(A=@n21, B=Constant() {r=3})) {g=true, r=true}
  n27 = ConstantBiasScale(None=TextureSample(UVs=Panner(Coordinate=@n26) {speedX=0.02, speedY=0.05}, Tex=@n23) {samplerSource=SSM_Wrap_WorldGroupSettings, samplerType=SAMPLERTYPE_LinearColor}.A) {bias=-0.5, scale=2}
  n28 = ScalarParameter() {defaultValue=0.0552, group=Caustics, parameterName=Caustic Warp Strength 2, sliderMax=0.1}
  NamedRerouteDeclaration(None=Add(A=Multiply(A=@n24, B=@n25), B=Multiply(A=@n27, B=@n28))) {name=Warp, nodeColor=<rgba 1 0 0.3281 1>}
  n29 = Divide(A=NamedRerouteUsage(), B=ScalarParameter() {defaultValue=0.6501, group=Caustics, parameterName=CausticScaleOffset})
  n30 = NamedRerouteUsage()
  n31 = Panner(Coordinate=Add(A=ComponentMask(None=@n29) {g=true, r=true}, B=@n30)) {speedX=-0.05, speedY=0.02}
  n32 = NamedRerouteUsage()
  n33 = TextureSample(UVs=@n31, Tex=@n32) {samplerSource=SSM_Wrap_WorldGroupSettings, samplerType=SAMPLERTYPE_LinearColor}
  n34 = ComponentMask(None=@n29) {g=true, r=true}
  n35 = OneMinus(None=Panner(Coordinate=Add(A=@n30, B=@n34)) {speedX=0.02, speedY=-0.02})
  n36 = TextureSample(UVs=@n35, Tex=@n32) {constCoordinate=1, samplerSource=SSM_Wrap_WorldGroupSettings, samplerType=SAMPLERTYPE_LinearColor}
  n37 = NamedRerouteUsage()
  n38 = TextureSample(UVs=@n31, Tex=@n37) {samplerSource=SSM_Wrap_WorldGroupSettings, samplerType=SAMPLERTYPE_LinearColor}
  n39 = TextureSample(UVs=@n35, Tex=@n37) {constCoordinate=1, samplerSource=SSM_Wrap_WorldGroupSettings, samplerType=SAMPLERTYPE_LinearColor}
  n40 = Subtract(A=WorldPosition().Z, B=ComponentMask(None=NamedRerouteUsage()) {b=true})
  n41 = Saturate(None=Divide(A=OneMinus(None=@n40), B=ScalarParameter() {defaultValue=420, parameterName=water depth}) {constB=100})
  n42 = Constant() {r=4}
  NamedRerouteDeclaration(None=LinearInterpolate(A=Add(A=@n33.RGB, B=@n36.RGB), B=Add(A=@n38.RGB, B=@n39.RGB), Alpha=Power(Base=@n41, Exp=@n42))) {name=Caustics, nodeColor=<rgba 0 1 0.4609 1>}
  NamedRerouteUsage()
 ]
}
