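# Material function that produces animated water caustics from world-position UVs for the Item Shop frontend art.
# WaterFlowMap warps UVs by water velocity, sampling a 6x6 caustics flipbook atlas with explicit derivatives and adjacent-frame blending.
# Depth and camera distance fade the result; ColoredResult skips the distance fade, while the main output becomes 1 on low quality and mobile.
# The fluid-simulation caustics and second flow-phase blend branches reach neither output.
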MaterialFunction /Game/UI/Frontend/ItemShop/Art/Ensemble/CT_MF_WaterCaustics
graph {nodes [
  n1 = FunctionInput() {bUsePreviewValueAsDefault=true, inputName=Scale, inputType=FunctionInput_Scalar, sortPriority=1, previewValue=<1024 0 0 1>}
  n2 = ScalarParameter() {defaultValue=0.5, group=Caustics, parameterName=Caustics Phase Fade Time}
  n3 = Divide(
   A=Divide(A=FunctionInput() {bUsePreviewValueAsDefault=true, inputName=Water Velocity, inputType=FunctionInput_Vector2, sortPriority=5}, B=@n1)
   B=@n2
  )
  n4 = FunctionInput(Preview=WorldPosition().XYZ) {bUsePreviewValueAsDefault=true, inputName=WorldPosition}
  n5 = ComponentMask(None=@n4) {g=true, r=true}
  n6 = Divide(A=@n5, B=@n1)
  n7 = MaterialFunctionCall!(
   Diffuse
   Normal Map
   "Flow Vector Map (see tooltip)"=@n3
   UV texture 2 offset and Scale=VectorParameter() {parameterName=Flowmap UV2 Offset and Scale, defaultValue=<rgba 0 0 0.85 0>}.RGB
   "Time (see tooltip)"=MaterialFunctionCall(Speed=@n2) {materialFunction=&TimeWithSpeedVariable}.Time
   UVs=@n6
  ) {materialFunction=/FortWater/Materials/Functions/WaterFlowMap}
  n8 = Constant() {r=6}
  n9 = Divide(A=Frac(None=@n7.UV A), B=@n8)
  n10 = Constant() {r=6}
  n11 = FunctionInput(Preview=MaterialFunctionCall(Speed=ScalarParameter() {defaultValue=0.5, group=Caustics, parameterName=WaterCausticsTimeSpeed}) {materialFunction=&TimeWithSpeedVariable}.Time) {bUsePreviewValueAsDefault=true, inputName=Time, inputType=FunctionInput_Scalar, sortPriority=3, previewValue=<1024 0 0 1>}
  n12 = Add(A=@n11) {constB=0.5}
  n13 = MaterialFunctionCall!(1D Index=Multiply(A=Multiply(A=@n10, B=@n10), B=Frac(None=@n12)), Cells XY=@n10) {materialFunction=/Engine/Functions/Engine_MaterialFunctions01/Coordinates/1Dto2DIndex}
  n14 = TextureObject() {samplerType=SAMPLERTYPE_LinearColor, texture=/FortWater/Textures/T_Siren_Caustics_03}
  n15 = Divide(A=@n6) {constB=6}
  n16 = DDX(Value=@n15)
  n17 = DDY(Value=@n15)
  n18 = TextureSample!(UVs=Add(A=@n9, B=Divide(A=@n13.2D Index, B=@n10)), Tex=@n14, "DDX(UVs)"=@n16, "DDY(UVs)"=@n17, Apply View MipBias) {mipValueMode=TMVM_Derivative}
  n19 = Frac(None=@n7.UV B)
  n20 = Constant() {r=6}
  n21 = MaterialFunctionCall(1D Index=Multiply(A=Multiply(A=@n20, B=@n20), B=Frac(None=@n11)), Cells XY=@n20) {materialFunction=&1Dto2DIndex}
  n22 = TextureSample(UVs=Add(A=Divide(A=@n19, B=@n8), B=Divide(A=@n21.2D Index, B=@n20)), Tex=@n14, "DDX(UVs)"=@n16, "DDY(UVs)"=@n17) {mipValueMode=TMVM_Derivative}
  n23 = ComponentMask(None=@n7.Distortion) {b=true}
  n24 = WorldPosition()
  n25 = FunctionInput(Preview=Subtract(A=ComponentMask(None=@n24.XYZ) {b=true}, B=ComponentMask(None=@n4) {b=true})) {bUsePreviewValueAsDefault=true, inputName=Water Depth, inputType=FunctionInput_Scalar, sortPriority=2, previewValue=<1024 0 0 1>}
  n26 = VectorParameter() {parameterName=V, defaultValue=<rgba 0 -0.1 0 0>}
  n27 = Add(A=@n5, B=Multiply(A=@n25, B=ComponentMask(None=@n26.RGB) {g=true, r=true}))
  n28 = ComponentMask(None=CollectionParameter() {collection=/Water/Materials/MPC/MPC_Water, parameterName=SimLocation}) {g=true, r=true}
  n29 = Add(A=Divide(A=Subtract(A=@n27, B=@n28), B=CollectionParameter() {collection=&MPC_Water, parameterName=FluidSimSize})) {constB=0.5}
  n30 = TextureSampleParameter2D(UVs=Saturate(None=@n29)) {group=Fluid Simulation, parameterName=NormalAndHeight, samplerSource=SSM_Wrap_WorldGroupSettings, texture=&Black}
  n31 = DeriveNormalZ(InXY=ComponentMask(None=@n30.RGB) {g=true, r=true})
  n32 = Add(A=@n29, B=ComponentMask(None=VectorParameter() {parameterName=UV Offset, defaultValue=<rgba 0.01 0.01 0 0>}.RGB) {g=true, r=true})
  n33 = TextureSampleParameter2D(UVs=Saturate(None=@n32)) {group=Fluid Simulation, parameterName=NormalAndHeight, samplerSource=SSM_Wrap_WorldGroupSettings, texture=&Black}
  n34 = DeriveNormalZ(InXY=ComponentMask(None=@n33.RGB) {g=true, r=true})
  n35 = Subtract(A=ComponentMask(None=OneMinus(None=@n31)) {b=true}, B=ComponentMask(None=OneMinus(None=@n34)) {b=true})
  Add(
   A=LinearInterpolate(A=@n18.RGB, B=@n22.RGB, Alpha=@n23)
   B=Multiply(A=Abs(None=@n35), B=ScalarParameter() {defaultValue=1, parameterName=Simulation Caustics})
  )
  Clamp(None=@n32)
  ComponentMask(None=WorldPosition().XYZ) {g=true, r=true}
  n36 = Constant() {r=6}
  n37 = Multiply(A=Multiply(A=@n36, B=@n36), B=Frac(None=Add(A=@n11) {constB=0.5}))
  n38 = Divide(
   A=MaterialFunctionCall(1D Index=Add(A=@n37), Cells XY=@n36) {materialFunction=&1Dto2DIndex}.2D Index
   B=@n36
  )
  n39 = LinearInterpolate(
   A=@n18.RGB
   B=TextureSample(UVs=Add(A=@n9, B=@n38), Tex=@n14, "DDX(UVs)"=@n16, "DDY(UVs)"=@n17) {mipValueMode=TMVM_Derivative}.RGB
   Alpha=Frac(None=@n37)
  )
  n40 = ScalarParameter() {defaultValue=1, group=Caustics, parameterName=Caustics Brightness}
  n41 = ScalarParameter() {defaultValue=256, group=Caustics, parameterName=Caustics Gradient Length}
  n42 = LinearInterpolate(
   A=ScalarParameter() {defaultValue=1, group=Caustics, parameterName=Caustics Fadeout Brightness}
   B=Add(A=Multiply(A=@n39, B=@n40), B=ScalarParameter() {defaultValue=0.2, group=Caustics, parameterName=Caustics Add})
   Alpha=Saturate(None=Divide(A=@n25, B=@n41))
  ) {constA=0.75}
  n43 = ScalarParameter() {defaultValue=7500, group=Caustics, parameterName=Caustics Fade Distance}
  n44 = LinearInterpolate(
   A=@n42
   B=ScalarParameter() {defaultValue=1.5, parameterName=Caustics Distance brightness}
   Alpha=MaterialFunctionCall(Fade Length=@n43, Fade Offset=@n43) {materialFunction=&CameraDepthFade}.Result
  )
  n45 = Constant() {r=1}
  FunctionOutput(None=QualitySwitch(Default=FeatureLevelSwitch(Default=@n44, ES2=@n45, ES3_1=@n45, SM4_REMOVED=@n44), Low=@n45))
  FunctionOutput(None=@n42) {outputName=ColoredResult}
  Multiply(A=@n3)
  Multiply(A=@n15, B=ScalarParameter() {defaultValue=5, parameterName=MipScale})
  Power(Base=LinearInterpolate(A=@n22.R, B=@n18.R, Alpha=@n23), Exp=ScalarParameter() {defaultValue=2, parameterName=Exp})
  ScalarParameter()
 ]}
