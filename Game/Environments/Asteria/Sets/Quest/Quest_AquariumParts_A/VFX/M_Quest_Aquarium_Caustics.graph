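# Light function material that casts animated-style water caustics, named for the Asteria quest aquarium set.
# World position, offset by the camera vector times a per-axis vector parameter, feeds the shared CT_MF_WaterCaustics function along with scale and water depth parameters, so the pattern shifts with view direction.
# The colored result is multiplied by its own red channel, raised to an exponent parameter and clamped, to shape contrast.
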Material /Game/Environments/Asteria/Sets/Quest/Quest_AquariumParts_A/VFX/M_Quest_Aquarium_Caustics
graph {
 properties {materialDomain=MD_LightFunction, pixelDepthOffsetMode=PDOM_Legacy, usedShadingModels=Default Lit}
 nodes [
  n1 = CameraVectorWS()
  n2 = VectorParameter() {defaultValue=<rgba 800 400 100 1>}
  n3 = MaterialFunctionCall!(
   WorldPosition=Add(A=WorldPosition().XYZ, B=Multiply(A=@n1, B=@n2.RGB))
   Scale=ScalarParameter() {defaultValue=400, parameterName="Scale (S)"}
   Water Depth=ScalarParameter() {defaultValue=114.4253, parameterName="Water Depth (S)"}
   Time
   Water Velocity
  ) {materialFunction=/Game/UI/Frontend/ItemShop/Art/Ensemble/CT_MF_WaterCaustics}
  n4 = ComponentMask(None=@n3.ColoredResult) {r=true}
  n5 = ScalarParameter() {defaultValue=1.25, parameterName=Exp_2}
  Multiply(A=@n3.ColoredResult, B=Clamp(None=Power(Base=@n4, Exp=@n5)))
 ]
}
