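# Translucent unlit depth-fog material for the quest aquarium set, tinting what lies behind the surface by depth.
# A depth fade blends two colors, then scene-minus-pixel depth, scaled and power-shaped, blends toward a third before MF_FX_EmissiveAdjustments.
# The same depth factor blends two scalars, apparently for opacity.
# Quality switches zero both the depth fade and that scalar blend on Low and Medium, likely hiding the fog there.
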Material /Game/Environments/Asteria/Sets/Quest/Quest_AquariumParts_A/VFX/M_Quest_Aquarium_Depthfog
graph {
 properties {bCastRayTracedShadows=false, blendMode=BLEND_Translucent, pixelDepthOffsetMode=PDOM_Legacy, shadingModel=MSM_Unlit, usedShadingModels=Unlit}
 nodes [
  n1 = VectorParameter() {defaultValue=<rgba 0 0.0156 0.0141 1>}
  n2 = VectorParameter() {parameterName=Param_1, defaultValue=<rgba 0.0161 0.1222 0.1823 1>}
  n3 = DepthFade(FadeDistance=ScalarParameter() {defaultValue=100, parameterName=Param_3})
  n4 = Constant()
  n5 = QualitySwitch(Default=@n3, Low=@n4, High=@n3, Medium=@n4, Epic=@n3)
  n6 = Subtract(A=SceneDepth().R, B=PixelDepth().R)
  n7 = ScalarParameter() {defaultValue=200}
  n8 = Saturate(None=Power(Base=Divide(A=@n6, B=@n7), Exp=ScalarParameter() {defaultValue=1, parameterName=Param_4}) {constExponent=0.5})
  MaterialFunctionCall(Emissive=LinearInterpolate(
    A=LinearInterpolate(A=@n1.RGB, B=@n2.RGB, Alpha=@n5)
    B=VectorParameter() {parameterName=Param1, defaultValue=<rgba 0 0.0771 0.2 1>}.RGB
    Alpha=@n8
   )) {materialFunction=&MF_FX_EmissiveAdjustments}
  n9 = LinearInterpolate(A=ScalarParameter() {defaultValue=0.3, parameterName=Param_1}, B=ScalarParameter() {defaultValue=0.45, parameterName=Param_2}, Alpha=@n8) {constA=0.25}
  n10 = Constant()
  QualitySwitch(Default=@n9, Low=@n10, High=@n9, Medium=@n10, Epic=@n9)
 ]
}
