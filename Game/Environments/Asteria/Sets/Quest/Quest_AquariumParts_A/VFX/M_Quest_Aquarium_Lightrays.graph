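# Unlit additive Niagara particle material for light rays in the Aquarium quest set.
# Panning Perlin noise, with per-particle random speed and phase, is shaped by vertical gradient and horizontal falloff masks.
# Camera-distance fade and depth fade, the latter bypassed on Low and Medium quality, soften it before particle color and alpha apply.
# SplineThicken tapers the ray geometry using top and bottom width parameters, and emissive passes through a shared FX adjustment function.
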Material /Game/Environments/Asteria/Sets/Quest/Quest_AquariumParts_A/VFX/M_Quest_Aquarium_Lightrays
graph {
 properties {
  bCastRayTracedShadows=false
  bUsedWithNiagaraMeshParticles=true
  bUsedWithNiagaraSprites=true
  blendMode=BLEND_Additive
  pixelDepthOffsetMode=PDOM_Legacy
  shadingModel=MSM_Unlit
  usedShadingModels=Unlit
 }
 nodes [
  n1 = TextureCoordinate() {uTiling=0.25, vTiling=0}
  n2 = Time()
  n3 = ParticleRandom()
  n4 = Add(A=LinearInterpolate(A=@n2, B=Multiply(A=@n2) {constB=1.25}, Alpha=@n3), B=@n3)
  n5 = Add(A=TextureSample(UVs=Panner(Coordinate=@n1, Time=@n4) {speedX=0.1, speedY=0.03}) {texture=/Game/Global/Textures/Default/Noise/3xPerlin_M}.R) {constB=0.5}
  n6 = ComponentMask(None=TextureCoordinate()) {g=true}
  n7 = OneMinus(None=@n6)
  n8 = Multiply(
   A=Multiply(A=SmoothStep(Value=@n7) {constMax=0.1}, B=@n6)
   B=SphereMask(A=TextureCoordinate() {vTiling=0}, B=Constant2Vector() {r=0.5}, Radius=Constant() {r=0.5}, Hardness=Constant())
  )
  n9 = DepthFade() {fadeDistanceDefault=25}
  n10 = Constant() {r=1}
  n11 = Multiply(
   A=Multiply(A=Multiply(A=@n5, B=@n8), B=QualitySwitch(Default=@n9, Low=@n10, High=@n9, Medium=@n10, Epic=@n9))
   B=MaterialFunctionCall(Fade Length=Constant() {r=256}, Fade Offset=Constant() {r=128}) {materialFunction=&CameraDepthFade}.Result
  )
  n12 = ParticleColor()
  MaterialFunctionCall(Emissive=Multiply(A=@n11, B=@n12.RGB)) {materialFunction=&MF_FX_EmissiveAdjustments}
  MaterialFunctionCall(
   WidthBase=ScalarParameter() {defaultValue=60, parameterName=TopWidth}
   WidthTip=ScalarParameter() {defaultValue=180, parameterName=BottomWidth}
  ) {materialFunction=&SplineThicken}
  Multiply(A=@n11, B=@n12.A)
 ]
}
