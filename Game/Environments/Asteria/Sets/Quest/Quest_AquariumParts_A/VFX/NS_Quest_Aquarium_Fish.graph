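# Niagara effect for the quest aquarium set, filling a tank volume with schooling mesh fish and rising bubbles.
# Fully transparent FishMover particles drift on panning curl noise; burst-spawned Fish read their positions through a PositionReader for point attraction, orient to velocity and carry a random VisibilityTag alongside three mesh renderers.
# Scattered Bubbles and intermittent BubblesLine columns rise until culled outside the volume, with user parameters scaling spread, opacity and size.
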NiagaraSystem /Game/Environments/Asteria/Sets/Quest/Quest_AquariumParts_A/VFX/NS_Quest_Aquarium_Fish
dependencies {
 usedDataInterfaces [&NiagaraDataInterfaceCurve, &NiagaraDataInterfaceParticleRead]
 usedDynamicInputs []
 usedModules [
  &SystemState
  &self:SystemSpawnScript.SystemScriptSource.SystemScriptGraph.NiagaraNodeAssignment_0.SetVariables_#1
 ]
}
emitters {
 Bubbles {
  emitterUpdateScript [
   &EmitterState()
   &SpawnRate(SpawnRate=&RandomRangeFloat(Minimum=1, Maximum=10), Use Spawn Probability=false)
   /SetVariables(Emitter.LineLength:Vector3f=&RandomRangeVector(Minimum=<0 0 0>, Maximum=&MakeVector(X=$User.BubbleLineLength, Y=0, Z=0))) {isSetParameters=true, module=SetVariables_#2}
   /SetVariables(Emitter.UseParticleColor=true) {isSetParameters=true}
  ]
  particleSpawnScript [
   &InitializeParticle(
    Color Maximum=<rgba 0.2834 0.6114 1 1>
    Color Minimum=<rgba 0.3555 0.543 0.6406 1>
    Color Mode=::NewEnumerator2
    Lifetime Max=8
    Lifetime Min=5
    Lifetime Mode=::NewEnumerator1
    Sprite Size Mode=::NewEnumerator3
    Uniform Sprite Size=&FloatFromCurve(
     FloatCurve=DataInterface() {propertyValues {Curve {keys [<at 0.2 8 RCIM_Cubic>, {arriveTangent=27.6937, interpMode=RCIM_Cubic, tangentMode=RCTM_Break, time=1, value=14}]}}}
     CurveIndex=&RandomRangeFloat(Minimum=0, Maximum=1)
     Scale Curve=1.5
    )
    UsePositionOffset=false
    Write Material Random=true
   )
   &BoxLocation(
    Box Size=&MakeVector(
     X=&Multiply_Float(A=&RandomRangeFloat(Minimum=-1, Maximum=1), B=$User.BubbleLineLength)
     Y=&RandomRangeFloat(Minimum=-400, Maximum=400)
     Z=&RandomRangeFloat(Minimum=-100, Maximum=100)
    )
    MaskSpawnGroup=false
    Module.Override Seed=false
    Rotate Box=false
    Surface Only=false
   )
   &AddVelocity(Velocity=&RandomRangeVector(Minimum=<0 0 20>, Maximum=<1 1 100>))
   &ApplyInitialForces()
  ]
  particleUpdateScript [
   &ParticleState()
   &Drag(
    Drag=&FloatFromCurve(
     FloatCurve=DataInterface() {propertyValues {Curve {keys [
         {arriveTangent=-2.3383, interpMode=RCIM_Cubic, leaveTangent=-2.3383, tangentMode=RCTM_User, time=0, value=1}
         {interpMode=RCIM_Cubic, tangentMode=RCTM_Break, time=1, value=0}
        ]}}}
     Scale Curve=0.4
    )
    Use Linear Drag=true
    Use Rotational Drag=false
   )
   &AccelerationForce(Acceleration=&RandomRangeVector(Minimum=<0 0 2>, Maximum=<1 1 15>))
   &CurlNoiseForce(
    Noise Frequency=20
    Noise Strength=&FloatFromCurve(
     FloatCurve=DataInterface() {propertyValues {Curve {keys [<at 0 0.05 RCIM_Cubic>, {arriveTangent=1.862, interpMode=RCIM_Cubic, tangentMode=RCTM_Break, time=1, value=1}]}}}
     Scale Curve=75
    )
    Pan Noise=false
   )
   &SolveForcesAndVelocity(Clamp Velocity=false, Limit Acceleration=false)
   &ScaleColor(
    Scale Alpha=&FloatFromCurve(
     FloatCurve=DataInterface() {propertyValues {Curve {keys [<at 0 0.6 RCIM_Cubic>, {arriveTangent=1.2873, interpMode=RCIM_Cubic, tangentMode=RCTM_Break, time=1, value=1}]}}}
     Scale Curve=&Multiply_Float(A=1, B=$User.OpacityScalar)
    )
    Scale RGB=&Multiply_VectorByFloat(Vector=<1 1 1>, Float=5)
    ScaleA=true
    ScaleRGB=true
   )
   &ScaleSpriteSize(Scale Factor=&Vector2DFromFloat(Value=&FloatFromCurve(
      FloatCurve=DataInterface() {propertyValues {Curve {keys [
          <at 0 0.4 RCIM_Cubic>
          {arriveTangent=-0.01, interpMode=RCIM_Cubic, leaveTangent=-0.1507, tangentMode=RCTM_Break, time=0.3, value=1}
          {interpMode=RCIM_Cubic, tangentMode=RCTM_Break, time=1, value=0.2}
         ]}}}
      Scale Curve=&Multiply_Float(A=1, B=$User.SizeScalar)
     )))
   &KillParticlesInVolume(Box Size:Vector3f=<400 400 325>, Invert Volume=true, Kill Shape=::NewEnumerator1, Origin Offset:Vector3f=<0 0 150>)
  ]
  rendererClasses [&NiagaraSpriteRendererProperties]
  renderers [{rendererClass=&NiagaraSpriteRendererProperties}]
 }
 BubblesLine {
  emitterUpdateScript [
   &EmitterState(
    Life Cycle Mode=::NewEnumerator1
    Loop Delay=&RandomRangeFloat(Minimum=0.5, Maximum=1.5)
    Loop Duration=&RandomRangeFloat(Minimum=0.75, Maximum=2.5)
    Recalculate Duration Each Loop=true
    UseLoopDelay=true
   )
   &SpawnRate(SpawnRate=&RandomRangeFloat(Minimum=10, Maximum=30), Use Spawn Probability=false)
   /SetVariables(Emitter.LineLength:Vector3f=&RandomRangeVector(
     Minimum=&MakeVector(
      X=&Multiply_Float(A=-1, B=&Multiply_Float(A=$User.BubbleLineLength, B=0.5))
      Y=0
      Z=0
     )
     Maximum=&MakeVector(X=&Multiply_Float(A=$User.BubbleLineLength, B=0.5), Y=0, Z=0)
    )) {isSetParameters=true, module=SetVariables_#2}
   /SetVariables(Emitter.UseParticleColor=true) {isSetParameters=true}
  ]
  particleSpawnScript [
   &InitializeParticle(
    Color Maximum=<rgba 0.2834 0.6114 1 1>
    Color Minimum=<rgba 0.3555 0.543 0.6406 1>
    Color Mode=::NewEnumerator2
    Lifetime Max=10
    Lifetime Min=5
    Lifetime Mode=::NewEnumerator1
    Sprite Size Mode=::NewEnumerator3
    Uniform Sprite Size=&FloatFromCurve(
     FloatCurve=DataInterface() {propertyValues {Curve {keys [<at 0.2 8 RCIM_Cubic>, {arriveTangent=27.6937, interpMode=RCIM_Cubic, tangentMode=RCTM_Break, time=1, value=14}]}}}
     CurveIndex=&RandomRangeFloat(Minimum=0, Maximum=1)
     Scale Curve=1.5
    )
    UsePositionOffset=false
    Write Material Random=true
   )
   &InitializeParticle(
    Color Maximum=<rgba 0.3438 0.785 1 1>
    Color Minimum=<rgba 1 1 1 1>
    Color Mode=::NewEnumerator2
    Lifetime Max=10
    Lifetime Min=5
    Lifetime Mode=::NewEnumerator1
    Sprite Size Mode=::NewEnumerator3
    Uniform Sprite Size=&FloatFromCurve(
     FloatCurve=DataInterface() {propertyValues {Curve {keys [<at 0.2 8 RCIM_Cubic>, {arriveTangent=27.6937, interpMode=RCIM_Cubic, tangentMode=RCTM_Break, time=1, value=14}]}}}
     CurveIndex=&RandomRangeFloat(Minimum=0, Maximum=1)
     Scale Curve=1.5
    )
    UsePositionOffset=false
    Write Material Random=true
   ) {enabled=false, module=InitializeParticle001}
   &BoxLocation(Box Offset=$Emitter.LineLength, Box Size=<12 12 12>, MaskSpawnGroup=false, Module.Override Seed=false, Rotate Box=false, Surface Only=false)
   &AddVelocity(Velocity=&RandomRangeVector(Minimum=<0 0 20>, Maximum=<1 1 100>))
   &ApplyInitialForces()
  ]
  particleUpdateScript [
   &ParticleState()
   &Drag(
    Drag=&FloatFromCurve(
     FloatCurve=DataInterface() {propertyValues {Curve {keys [
         {arriveTangent=-2.3383, interpMode=RCIM_Cubic, leaveTangent=-2.3383, tangentMode=RCTM_User, time=0, value=1}
         {interpMode=RCIM_Cubic, tangentMode=RCTM_Break, time=1, value=0}
        ]}}}
     Scale Curve=0.4
    )
    Use Linear Drag=true
    Use Rotational Drag=false
   )
   &AccelerationForce(Acceleration=&RandomRangeVector(Minimum=<0 0 2>, Maximum=<1 1 15>))
   &CurlNoiseForce(
    Noise Strength=&FloatFromCurve(
     FloatCurve=DataInterface() {propertyValues {Curve {keys [<at 0 0.05 RCIM_Cubic>, {arriveTangent=1.862, interpMode=RCIM_Cubic, tangentMode=RCTM_Break, time=1, value=1}]}}}
     Scale Curve=150
    )
    Pan Noise=false
   )
   &SolveForcesAndVelocity(Clamp Velocity=false, Limit Acceleration=false)
   &ScaleColor(
    Scale Alpha=&FloatFromCurve(
     FloatCurve=DataInterface() {propertyValues {Curve {keys [<at 0 0.6 RCIM_Cubic>, {arriveTangent=1.2873, interpMode=RCIM_Cubic, tangentMode=RCTM_Break, time=1, value=1}]}}}
     Scale Curve=&Multiply_Float(A=1, B=$User.OpacityScalar)
    )
    Scale RGB=&Multiply_VectorByFloat(Vector=<1 1 1>, Float=5)
    ScaleA=true
    ScaleRGB=true
   )
   &ScaleSpriteSize(Scale Factor=&Vector2DFromFloat(Value=&FloatFromCurve(
      FloatCurve=DataInterface() {propertyValues {Curve {keys [
          <at 0 0.4 RCIM_Cubic>
          {arriveTangent=-0.01, interpMode=RCIM_Cubic, leaveTangent=-0.1507, tangentMode=RCTM_Break, time=0.3, value=1}
          {interpMode=RCIM_Cubic, tangentMode=RCTM_Break, time=1, value=0.2}
         ]}}}
      Scale Curve=&Multiply_Float(A=1, B=$User.SizeScalar)
     )))
   &KillParticlesInVolume(Box Size:Vector3f=<400 400 325>, Invert Volume=true, Kill Shape=::NewEnumerator1, Origin Offset:Vector3f=<0 0 150>)
  ]
  rendererClasses [&NiagaraSpriteRendererProperties]
  renderers [{rendererClass=&NiagaraSpriteRendererProperties}]
 }
 Fish {
  emitterUpdateScript [
   &EmitterState()
   &SpawnRate(SpawnRate=5, Use Spawn Probability=false) {enabled=false}
   &SpawnBurst_Instantaneous(Spawn Count=&RandomRangeInt(Minimum=20, Maximum=40), Use Spawn Probability=false)
  ]
  particleSpawnScript [
   &InitializeParticle(
    Lifetime Max=8
    Lifetime Min=5
    Lifetime Mode=::NewEnumerator1
    Mesh Scale Mode=::NewEnumerator1
    UsePositionOffset=false
    Write Material Random=true
   )
   &BoxLocation(
    Box Offset=&MakeVector(X=0, Y=0, Z=150)
    Box Size=<150 150 150>
    MaskSpawnGroup=false
    Module.Override Seed=false
    Rotate Box=false
    Surface Only=false
   )
   &ApplyInitialForces()
   /SetVariables(Particles.VisibilityTag=&RandomRangeInt(Minimum=0, Maximum=2)) {isSetParameters=true}
  ]
  particleUpdateScript [
   &ParticleState()
   &self:PositionReader(
    IndexID:NiagaraInt32=&RandomRangeInt(Minimum=0, Maximum=1)
    New Particle Attribute Reader=DataInterface() {propertyValues {EmitterBinding {bindingMode=Other, emitterName=FishMover}}}
   ) {module=PositionReader}
   &CurlNoiseForce(Noise Strength=100, Pan Noise=false)
   &Drag(Drag=&RandomRangeFloat(Minimum=0.1, Maximum=0.5), Use Linear Drag=true, Use Rotational Drag=false)
   &PointAttractionForce(
    Attraction Radius=10000
    Attraction Strength=&RandomRangeFloat(Minimum=1, Maximum=8)
    Attractor Position:Vector3f=$Particles.FishMoverPosition
    Falloff Exponent=0.8
    Kill Within Radius=false
    Module.Use Falloff=true
    UseAttractorOffset=false
   )
   &SolveForcesAndVelocity(Clamp Velocity=false, Limit Acceleration=false)
   /Niagara/Modules/Update/Orientation/V2/OrientMeshToVector(Look At Direction:Vector3f=$Particles.Velocity, Override Quat=false, UseOptionalSideAxis=false, UseOptionalUpAxis=false)
   &ScaleMeshSize(Scale Factor=&Multiply_VectorByFloat(
     Vector=<1 1 1>
     Float=&FloatFromCurve(
      FloatCurve=DataInterface() {propertyValues {Curve {keys [
          <at 0 0 RCIM_Cubic>
          {interpMode=RCIM_Cubic, tangentMode=RCTM_Break, time=0.2, value=1}
          {interpMode=RCIM_Cubic, tangentMode=RCTM_Break, time=0.8, value=1}
          <at 1 0 RCIM_Cubic>
         ]}}}
      Scale Curve=2
     )
    ))
  ]
  rendererClasses [&NiagaraMeshRendererProperties]
  renderers [
   {rendererClass=&NiagaraMeshRendererProperties, rendererIndex=0}
   {rendererClass=&NiagaraMeshRendererProperties, rendererIndex=1}
   {rendererClass=&NiagaraMeshRendererProperties, rendererIndex=2}
  ]
 }
 FishMover {
  emitterUpdateScript [
   &EmitterState()
   &SpawnRate(SpawnRate=2, Use Spawn Probability=false)
  ]
  particleSpawnScript [
   &InitializeParticle(
    Color=<rgba 1 1 1 0>
    Lifetime=4
    Sprite Size Mode=::NewEnumerator3
    Uniform Sprite Size=1
    UsePositionOffset=false
    Write Material Random=true
   )
   &BoxLocation(
    Box Offset=&MakeVector(X=0, Y=0, Z=150)
    Box Size=<150 5 5>
    MaskSpawnGroup=false
    Module.Override Seed=false
    Rotate Box=false
    Surface Only=false
   )
   &AddVelocity(Scale Added Velocity=<0 0 0>, Velocity=&RandomRangeVector(Minimum=<-50 -10 -50>, Maximum=<-50 10 50>))
  ]
  particleUpdateScript [
   &ParticleState()
   &CurlNoiseForce(Pan Noise=true, Pan Noise Field=<0 0 0.5>)
   &SolveForcesAndVelocity(Clamp Velocity=false, Limit Acceleration=false)
  ]
  rendererClasses [&NiagaraSpriteRendererProperties]
  renderers [{rendererClass=&NiagaraSpriteRendererProperties}]
 }
}
summary {
 systemName=NS_Quest_Aquarium_Fish
 emitters [
  {emitterName=Bubbles, rendererClasses [&NiagaraSpriteRendererProperties]}
  {emitterName=BubblesLine, rendererClasses [&NiagaraSpriteRendererProperties]}
  {emitterName=Fish, rendererClasses [&NiagaraMeshRendererProperties]}
  {emitterName=FishMover, rendererClasses [&NiagaraSpriteRendererProperties]}
 ]
 userVariables [
  User.BubbleLineLength {type=NiagaraFloat, defaultValue=400, description=""}
  User.FishOffset {type=NiagaraFloat, defaultValue=-200, description=""}
  User.OpacityScalar {type=NiagaraFloat, defaultValue=0.1, description=""}
  User.SizeScalar {type=NiagaraFloat, defaultValue=1, description=""}
 ]
}
