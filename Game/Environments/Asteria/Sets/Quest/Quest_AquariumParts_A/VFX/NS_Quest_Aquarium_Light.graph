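# Niagara system for the aquarium quest set piece that draws a fixed 4x4 grid of blue light rays with glows.
# One Rays emitter bursts 16 persistent particles, placed by execution index from an offset array and drawn by sprite and mesh renderers.
# Each particle stores a glow position just below itself and a glow color that LOD distance scales, dimming it at close range.
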NiagaraSystem /Game/Environments/Asteria/Sets/Quest/Quest_AquariumParts_A/VFX/NS_Quest_Aquarium_Light
dependencies {
 usedDataInterfaces [&NiagaraDataInterfaceArrayFloat3]
 usedDynamicInputs []
 usedModules [&SystemState]
}
emitters {Rays {
  emitterUpdateScript [
   &EmitterState(Life Cycle Mode=::NewEnumerator1, Loop Behavior=::NewEnumerator1, Loop Duration Mode=::NewEnumerator1, UseLoopDelay=false)
   &SpawnBurst_Instantaneous(Spawn Count=16, Use Spawn Probability=false, UseLoopCountLimit=false)
  ]
  particleSpawnScript [
   &InitializeParticle(
    Color Mode=::NewEnumerator0
    Lifetime=5
    Mesh Scale=<4 1 1>
    Mesh Scale Mode=::NewEnumerator4
    Ribbon Facing Mode=::NewEnumerator1
    Ribbon Facing Vector:Vector3f=<0 1 0>
    Sprite Rotation Mode=::NewEnumerator2
    Sprite Size Mode=::NewEnumerator3
    Uniform Sprite Size=150
    UsePositionOffset=false
    Write Material Random=true
   )
   &CameraOffset(Camera Offset Amount=25)
   &SystemLocation(
    MaskSpawnGroup=false
    Offset=&SelectVectorFromArray(
     Direct Array Index=&ReturnExecIndex()
     Vector Selection Array=DataInterface() {propertyValues {FloatData [
        <150 172.5 35>
        <150 67.5 35>
        <150 -62.5 35>
        <150 -167.5 35>
        <50 172.5 35>
        <50 67.5 35>
        <50 -62.5 35>
        <50 -167.5 35>
        <-51 172.5 35>
        <-51 67.5 35>
        <-51 -62.5 35>
        <-51 -167.5 35>
        <-157 172.5 35>
        <-157 67.5 35>
        <-157 -62.5 35>
        <-157 -167.5 35>
       ]}}
    )
   )
   &Color(Color=<rgba 0 0.4358 1 0.3>)
   &InitialMeshOrientation(Mesh Axis To Orient=<0 0 1>, Mesh Orientation Mode=::NewEnumerator2, Use Rotation Vector=false)
   /SetVariables(
    Particles.GlowColorInit:LinearColor=<rgba 0.3594 0.7109 1 1.5>
    Particles.Glowpos:NiagaraPosition=&AddVectorToPosition(Position=$Particles.Position, Vector=<0 0 -20>)
   ) {isSetParameters=true}
  ]
  particleUpdateScript [
   &ParticleState(Kill Particles When Lifetime Has Elapsed=false, Let Infinitely Lived Particles Die When Emitter Deactivates=true)
   /SetVariables(Particles.GlowColor=&MultiplyLinearColor(
     Linear Color=$Particles.GlowColorInit
     Scale Factor=&MakeVector4_FromFloat(Float=&Lerp_Float(
       A=0
       B=1
       Alpha=&FN_LOD_Distance_Remap(Start Distance=0, End Distance=1000, Minimum Output=0, Maximum Output=1, Clamp Output=true)
      ))
    )) {isSetParameters=true}
  ]
  rendererClasses [&NiagaraSpriteRendererProperties, &NiagaraMeshRendererProperties]
  renderers [
   {rendererClass=&NiagaraSpriteRendererProperties}
   {rendererClass=&NiagaraMeshRendererProperties, rendererIndex=1}
  ]
 }}
summary {
 systemName=NS_Quest_Aquarium_Light
 emitters [{emitterName=Rays, rendererClasses [&NiagaraSpriteRendererProperties, &NiagaraMeshRendererProperties]}]
 userVariables [User.RibbonIDOffset {type=NiagaraInt32}]
}
